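# BowlVenus capture-point sound cue for contested claim progress, playing a start sting plus a looping bed, limited to one concurrent instance.
# A switch on PlayerAffliation selects a mix with an added non-spatialized 2D start layer for one affiliation, and an attenuated-only mix otherwise.
# In both mixes the loop fades in through an enveloper, is pitched down by a modulator, and shares the start sting's POI capture-point attenuation.
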SoundCue /BowlVenus/Assets/Audio/BowlVenus_Claim_Progress_Contested_Loop_Cue
root=Switch_1
nodes [
 {id=Attenuation_0, children [WavePlayer_2], properties {attenuationSettings=/POIBanner/Sounds/POI_Capture_Point_Att}}
 {id=Attenuation_1, children [Modulator_0], properties {attenuationSettings=&POI_Capture_Point_Att}}
 {id=Attenuation_3, children [Modulator_1], properties {attenuationSettings=&POI_Capture_Point_Att}}
 {id=Attenuation_4, children [WavePlayer_6], properties {attenuationSettings=&POI_Capture_Point_Att}}
 {
  id=Enveloper_0
  children [WavePlayer_3]
  properties {pitchCurve {editorCurveData {keys [{time=0}]}}, volumeCurve {editorCurveData {keys [<at 0 0>,<at 1 1>]}}}
 }
 {
  id=Enveloper_1
  children [WavePlayer_5]
  properties {pitchCurve {editorCurveData {keys [{time=0}]}}, volumeCurve {editorCurveData {keys [<at 0 0>,<at 1 1>]}}}
 }
 {id=Mixer_0, children [WavePlayer_0, Attenuation_0, Attenuation_1], properties {inputVolume [0.35,1,0.25]}}
 {id=Mixer_1, children [Attenuation_4, Attenuation_3], properties {inputVolume [2.2,0.25]}}
 {id=Modulator_0, children [Enveloper_0], properties {pitchMax=0.7, pitchMin=0.7, volumeMax=1, volumeMin=1}}
 {id=Modulator_1, children [Enveloper_1], properties {pitchMax=0.7, pitchMin=0.7, volumeMax=1, volumeMin=1}}
 {id=Switch_1, children [Mixer_1,Mixer_0,Mixer_1,Mixer_1], properties {intParameterName=PlayerAffliation}}
 {id=WavePlayer_0, properties {soundWaveAssetPtr=&/CampBartFlag_Claim_Progress_Contested_Start_2D}}
 {id=WavePlayer_2, properties {soundWaveAssetPtr=&/CampBartFlag_Claim_Progress_Contested_Start}}
 {id=WavePlayer_3, properties {bLooping=true, soundWaveAssetPtr=&/BowlVenus_Claim_Progress_Contested_Loop}}
 {id=WavePlayer_5, properties {bLooping=true, soundWaveAssetPtr=&BowlVenus_Claim_Progress_Contested_Loop}}
 {id=WavePlayer_6, properties {soundWaveAssetPtr=&CampBartFlag_Claim_Progress_Contested_Start}}
]
properties {
 bOverrideConcurrency=true
 maxDistance=14500
 soundClassObject=&Fort_Gadgets
 volumeMultiplier=1.2
 concurrencyOverrides {maxCount=1}
}
