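# Sound cue for BowlVenus that, by its name, plays when the contested claim-progress loop stops.
# A Random node picks one of two equally weighted CampBartFlag claim-progress stop waves, then a Modulator randomizes pitch at nearly double the original.
# It uses POI capture point attenuation and the Fort_Gadgets sound class at reduced volume, with a concurrency override allowing one instance at a time.
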SoundCue /BowlVenus/Assets/Audio/BowlVenus_Claim_Progress_Contested_Loop_Stop_Cue
root=Modulator_1
nodes [
 {id=Modulator_1, children [Random_0], properties {pitchMax=2, pitchMin=1.8, volumeMax=1, volumeMin=1}}
 {id=Random_0, children [WavePlayer_3,WavePlayer_4], properties {bShouldExcludeFromBranchCulling=true, weights [1,1]}}
 {id=WavePlayer_3, properties {soundWaveAssetPtr=&/CampBartFlag_Claim_Progress_Stop_01}}
 {id=WavePlayer_4, properties {soundWaveAssetPtr=&/CampBartFlag_Claim_Progress_Stop_02}}
]
properties {
 attenuationSettings=/POIBanner/Sounds/POI_Capture_Point_Att
 bOverrideConcurrency=true
 maxDistance=14500
 soundClassObject=&Fort_Gadgets
 volumeMultiplier=0.6
 concurrencyOverrides {maxCount=1}
}
