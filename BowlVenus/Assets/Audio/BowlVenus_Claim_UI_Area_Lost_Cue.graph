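# UI sound cue in BowlVenus for the claim "area lost" notification.
# A single Wave Player plays the CampBartFlag area-lost wave directly, with a slight volume boost, routed to the non-critical in-game UI sound class.
# A concurrency override limits it to one instance, stopping the oldest when retriggered.
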SoundCue /BowlVenus/Assets/Audio/BowlVenus_Claim_UI_Area_Lost_Cue
root=WavePlayer_0
nodes [{id=WavePlayer_0, properties {soundWaveAssetPtr=&/CampBartFlag_Claim_UI_Area_Lost}}]
properties {
 bOverrideConcurrency=true
 maxDistance=2097152
 soundClassObject=&Fort_UI_InGame_NonCritical
 volumeMultiplier=1.2
 concurrencyOverrides {maxCount=1, resolutionRule=StopOldest}
}
