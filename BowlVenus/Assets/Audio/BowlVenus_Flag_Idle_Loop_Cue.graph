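# Sound Cue that supplies a looping idle sound for a BowlVenus flag.
# A single looping Wave Player is the root and plays the CampBartFlag idle loop wave.
# It uses the POI capture point pole attenuation asset and the Fort_Gadgets sound class.
# A concurrency override limits it to one playing instance.
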SoundCue /BowlVenus/Assets/Audio/BowlVenus_Flag_Idle_Loop_Cue
root=WavePlayer_3
nodes [{id=WavePlayer_3, properties {bLooping=true, soundWaveAssetPtr=&/CampBartFlag_Idle_Loop}}]
properties {
 attenuationSettings=/POIBanner/Sounds/POI_Capture_Point_Pole_Att
 bOverrideConcurrency=true
 maxDistance=13500
 soundClassObject=&Fort_Gadgets
 volumeMultiplier=1
 attenuationOverrides {falloffDistance=8000}
 concurrencyOverrides {maxCount=1}
}
